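# Physics asset for the M_MED_Shark_Backpack cosmetic back bling, a Fortnite backpack accessory.
# It defines a single capsule body on the backpack_root bone, offset and rotated from that bone.
# It has no constraints, so the backpack is represented as one simple collision shape with no jointed or secondary physics bodies.
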PhysicsAsset /BRCosmetics/Accessories/FORT_Backpacks/M_MED_Shark_Backpack/Meshes/M_MED_Shark_Backpack_PhysicsAsset
bodies {backpack_root [{shapeType=Capsule, length=17.2429, radius=11.1845, center=<-5.7882 -8.7446 0>, rotation=<rot 0 159.2616 -90>}]}
constraints []
